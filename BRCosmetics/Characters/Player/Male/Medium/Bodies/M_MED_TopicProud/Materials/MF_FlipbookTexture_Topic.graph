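# Material function for the TopicProud character body that samples a chosen frame from a flipbook texture, outputting color, alpha and UVs.
# UVs, defaulting to channel 1, are offset, center-scaled and saturated before feeding MF_FlipbookSubUV with the index and grid size.
# An optional static switch mirrors U on one side, split by interpolated pre-skinned local X.
# Two sphere masks on the sub-UV function's returned U and V fade alpha toward edges.
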MaterialFunction /BRCosmetics/Characters/Player/Male/Medium/Bodies/M_MED_TopicProud/Materials/MF_FlipbookTexture_Topic
graph {expressions {
  MaterialExpressionAdd_0 {
   class=MaterialExpressionAdd
   inputs [
    A {from=MaterialExpressionFunctionInput_9}
    B {from=MaterialExpressionFunctionInput_1}
   ]
  }
  MaterialExpressionAdd_2 {
   class=MaterialExpressionAdd
   inputs [
    A {from=MaterialExpressionConvert_2}
    B {from=MaterialExpressionFunctionInput_1}
   ]
  }
  MaterialExpressionAppendVector_1 {
   class=MaterialExpressionAppendVector
   inputs [
    A {from=MaterialExpressionFunctionInput_3, output=""}
    B {from=MaterialExpressionFunctionInput_4, output=""}
   ]
  }
  MaterialExpressionComponentMask_6 {class=MaterialExpressionComponentMask, inputs [None {from=MaterialExpressionVertexInterpolator_1, output=PS}], properties {r=true}}
  MaterialExpressionConstant_11 {class=MaterialExpressionConstant, properties {r=0.45}}
  MaterialExpressionConvert_1 {
   class=MaterialExpressionConvert
   inputs [None {from=MaterialExpressionFunctionInput_9}]
   properties {
    nodeName=Break Float2
    convertInputs [{
      type=Vector2
      expressionInput {
       expression=&self:MaterialExpressionFunctionInput_9
       inputName=None
       mask=0
       maskA=0
       maskB=0
       maskG=0
       maskR=0
       outputIndex=0
      }
     }]
    convertOutputs [{type=Scalar}, {type=Scalar}]
   }
  }
  MaterialExpressionConvert_2 {
   class=MaterialExpressionConvert
   inputs [
    None {from=MaterialExpressionOneMinus_0}
    None {from=MaterialExpressionConvert_1}
   ]
   properties {
    nodeName=Make Vector2
    convertInputs [
     {
      type=Scalar
      expressionInput {
       expression=&self:MaterialExpressionOneMinus_0
       inputName=None
       mask=0
       maskA=0
       maskB=0
       maskG=0
       maskR=0
       outputIndex=0
      }
     }
     {
      type=Scalar
      expressionInput {
       expression=&self:MaterialExpressionConvert_1
       inputName=None
       mask=0
       maskA=0
       maskB=0
       maskG=0
       maskR=0
       outputIndex=1
      }
     }
    ]
    convertOutputs [{type=Vector2}]
   }
  }
  MaterialExpressionFunctionInput_0 {
   class=MaterialExpressionFunctionInput
   inputs [Preview {from="", output=""}]
   properties {bUsePreviewValueAsDefault=true, inputName=UV Scale, inputType=FunctionInput_Vector2, sortPriority=4, previewValue=<1 1 0 1>}
  }
  MaterialExpressionFunctionInput_1 {
   class=MaterialExpressionFunctionInput
   inputs [Preview {from="", output=""}]
   properties {bUsePreviewValueAsDefault=true, inputName=UV Offset, inputType=FunctionInput_Vector2, sortPriority=4, previewValue=<0.3 0.5 0 1>}
  }
  MaterialExpressionFunctionInput_2 {
   class=MaterialExpressionFunctionInput
   inputs [Preview {from="", output=""}]
   properties {bUsePreviewValueAsDefault=true, inputName=Flipbook Index, inputType=FunctionInput_Scalar, sortPriority=-10}
  }
  MaterialExpressionFunctionInput_3 {
   class=MaterialExpressionFunctionInput
   inputs [Preview {from="", output=""}]
   properties {bUsePreviewValueAsDefault=true, inputName=Flipbook Column Count, inputType=FunctionInput_Scalar, sortPriority=1, previewValue=<4 0 0 1>}
  }
  MaterialExpressionFunctionInput_4 {
   class=MaterialExpressionFunctionInput
   inputs [Preview {from="", output=""}]
   properties {bUsePreviewValueAsDefault=true, inputName=Flipbook Row Count, inputType=FunctionInput_Scalar, sortPriority=2, previewValue=<4 0 0 1>}
  }
  MaterialExpressionFunctionInput_5 {
   class=MaterialExpressionFunctionInput
   inputs [Preview {from=MaterialExpressionTextureObject_0, output=""}]
   properties {bUsePreviewValueAsDefault=true, inputName=Flipbook Texture, inputType=FunctionInput_Texture2D}
  }
  MaterialExpressionFunctionInput_8 {
   class=MaterialExpressionFunctionInput
   inputs [Preview {from=MaterialExpressionStaticBool_1, output=""}]
   properties {bUsePreviewValueAsDefault=true, inputName=MirrorLeftUV, inputType=FunctionInput_StaticBool, sortPriority=7}
  }
  MaterialExpressionFunctionInput_9 {
   class=MaterialExpressionFunctionInput
   inputs [Preview {from=MaterialExpressionTextureCoordinate_0, output=""}]
   properties {bUsePreviewValueAsDefault=true, inputName=UVs, inputType=FunctionInput_Vector2, sortPriority=3}
  }
  MaterialExpressionFunctionOutput_0 {class=MaterialExpressionFunctionOutput, inputs [None {from=MaterialExpressionReroute_0, output=""}], properties {outputName=Color}}
  MaterialExpressionFunctionOutput_1 {
   class=MaterialExpressionFunctionOutput
   inputs [None {from=MaterialExpressionNamedRerouteUsage_3, output=""}]
   properties {outputName=UV_OUT, sortPriority=10}
  }
  MaterialExpressionFunctionOutput_5 {
   class=MaterialExpressionFunctionOutput
   inputs [None {from=MaterialExpressionMultiply_0, output=""}]
   properties {outputName=Alpha, sortPriority=1}
  }
  MaterialExpressionLinearInterpolate_1 {
   class=MaterialExpressionLinearInterpolate
   inputs [
    A {from=MaterialExpressionNamedRerouteUsage_1}
    B {from=MaterialExpressionNamedRerouteUsage_4}
    Alpha {from=MaterialExpressionStep_1}
   ]
  }
  MaterialExpressionLocalPosition_1 {class=MaterialExpressionLocalPosition, inputs [], properties {localOrigin=InstancePreSkinning}}
  MaterialExpressionMaterialFunctionCall_0 {
   class=MaterialExpressionMaterialFunctionCall
   inputs [
    UVs {from=MaterialExpressionAdd_0}
    Texture Scale {from=MaterialExpressionFunctionInput_0}
   ]
   properties {materialFunction=&ScaleUVsByCenter}
  }
  MaterialExpressionMaterialFunctionCall_1 {
   class=MaterialExpressionMaterialFunctionCall
   inputs [
    Texture {from=MaterialExpressionFunctionInput_5}
    UVs {from=MaterialExpressionNamedRerouteUsage_3}
    SubImages {from=MaterialExpressionAppendVector_1}
    Frame {from=MaterialExpressionFunctionInput_2}
   ]
   properties {materialFunction=/Game/Global/Materials/Character_Parent_2023/FlipbookFace_Parent_2023/MaterialFunctions/Common/MF_FlipbookSubUV}
  }
  MaterialExpressionMaterialFunctionCall_2 {
   class=MaterialExpressionMaterialFunctionCall
   inputs [
    UVs {from=MaterialExpressionAdd_2}
    Texture Scale {from=MaterialExpressionFunctionInput_0}
   ]
   properties {materialFunction=&ScaleUVsByCenter}
  }
  MaterialExpressionMaterialFunctionCall_36 {
   class=MaterialExpressionMaterialFunctionCall
   inputs [Float2 {from=MaterialExpressionMaterialFunctionCall_1, output=UVs}]
   properties {materialFunction=&BreakOutFloat2Components}
  }
  MaterialExpressionMultiply_0 {
   class=MaterialExpressionMultiply
   inputs [
    A {from=MaterialExpressionMaterialFunctionCall_1, output=Alpha}
    B {from=MaterialExpressionMultiply_13}
   ]
  }
  MaterialExpressionMultiply_13 {
   class=MaterialExpressionMultiply
   inputs [
    A {from=MaterialExpressionSphereMask_30}
    B {from=MaterialExpressionSphereMask_29}
   ]
  }
  MaterialExpressionNamedRerouteDeclaration_0 {
   class=MaterialExpressionNamedRerouteDeclaration
   inputs [None {from=MaterialExpressionSaturate_1}]
   properties {name=ScaledUVs, nodeColor=<rgba 1 0.2578 0 1>}
  }
  MaterialExpressionNamedRerouteDeclaration_1 {
   class=MaterialExpressionNamedRerouteDeclaration
   inputs [None {from=MaterialExpressionStaticSwitch_1}]
   properties {name=UVs, nodeColor=<rgba 0 0.0625 1 1>}
  }
  MaterialExpressionNamedRerouteDeclaration_2 {
   class=MaterialExpressionNamedRerouteDeclaration
   inputs [None {from=MaterialExpressionSaturate_2}]
   properties {name=ScaledUVs_Mirrored, nodeColor=<rgba 1 0.1406 0 1>}
  }
  MaterialExpressionNamedRerouteUsage_0 {class=MaterialExpressionNamedRerouteUsage}
  MaterialExpressionNamedRerouteUsage_1 {class=MaterialExpressionNamedRerouteUsage}
  MaterialExpressionNamedRerouteUsage_3 {class=MaterialExpressionNamedRerouteUsage}
  MaterialExpressionNamedRerouteUsage_4 {class=MaterialExpressionNamedRerouteUsage}
  MaterialExpressionOneMinus_0 {class=MaterialExpressionOneMinus, inputs [None {from=MaterialExpressionConvert_1}]}
  MaterialExpressionReroute_0 {class=MaterialExpressionReroute, inputs [None {from=MaterialExpressionMaterialFunctionCall_1, output=RGB}]}
  MaterialExpressionSaturate_1 {class=MaterialExpressionSaturate, inputs [None {from=MaterialExpressionMaterialFunctionCall_0, output=UVs}]}
  MaterialExpressionSaturate_2 {class=MaterialExpressionSaturate, inputs [None {from=MaterialExpressionMaterialFunctionCall_2, output=UVs}]}
  MaterialExpressionSphereMask_29 {
   class=MaterialExpressionSphereMask
   inputs [
    A {from=MaterialExpressionMaterialFunctionCall_36, output=G}
    B {from=MaterialExpressionConstant_11, output=""}
    Radius {from="", output=""}
    Hardness {from="", output=""}
   ]
   properties {attenuationRadius=0.5}
  }
  MaterialExpressionSphereMask_30 {
   class=MaterialExpressionSphereMask
   inputs [
    A {from=MaterialExpressionConstant_11, output=""}
    B {from=MaterialExpressionMaterialFunctionCall_36, output=R}
    Radius {from="", output=""}
    Hardness {from="", output=""}
   ]
   properties {attenuationRadius=0.5}
  }
  MaterialExpressionStaticBool_1 {class=MaterialExpressionStaticBool}
  MaterialExpressionStaticSwitch_1 {
   class=MaterialExpressionStaticSwitch
   inputs [
    True {from=MaterialExpressionLinearInterpolate_1, output=""}
    False {from=MaterialExpressionNamedRerouteUsage_0, output=""}
    Value {from=MaterialExpressionFunctionInput_8, output=""}
   ]
   properties {}
  }
  MaterialExpressionStep_1 {
   class=MaterialExpressionStep
   inputs [
    Y {from="", output=""}
    X {from=MaterialExpressionComponentMask_6, output=""}
   ]
   properties {}
  }
  MaterialExpressionTextureCoordinate_0 {class=MaterialExpressionTextureCoordinate, properties {coordinateIndex=1}}
  MaterialExpressionTextureObject_0 {class=MaterialExpressionTextureObject, inputs [], properties {texture=&T_BlacksRGB}}
  MaterialExpressionVertexInterpolator_1 {class=MaterialExpressionVertexInterpolator, inputs [VS {from=MaterialExpressionLocalPosition_1, output=XY}], properties {}}
 }}
