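# Material function for the flipbook face parent material that maps an Index to one of up to 16 scalar values packed into four Vector4 inputs.
# Index divided by 4 and floored picks the vector, while Index modulo 4 picks the component, both through Switch nodes.
# Static bools extend the table from the default 8 entries to 12 or 16, with the 16-input option taking precedence.
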MaterialFunction /Game/Global/Materials/Character_Parent_2023/FlipbookFace_Parent_2023/MaterialFunctions/Common/MF_FlipbookIndexSelector
graph {expressions {
  MaterialExpressionConstant_0 {class=MaterialExpressionConstant, properties {r=4}}
  MaterialExpressionDivide_0 {
   class=MaterialExpressionDivide
   inputs [
    A {from=MaterialExpressionNamedRerouteDeclaration_0}
    B {from=""}
   ]
   properties {constB=4}
  }
  MaterialExpressionFloor_0 {class=MaterialExpressionFloor, inputs [None {from=MaterialExpressionDivide_0, output=""}], properties {}}
  MaterialExpressionFmod_0 {
   class=MaterialExpressionFmod
   inputs [
    A {from=MaterialExpressionNamedRerouteUsage_0, output=""}
    B {from=MaterialExpressionConstant_0, output=""}
   ]
   properties {}
  }
  MaterialExpressionFunctionInput_0 {
   class=MaterialExpressionFunctionInput
   inputs [Preview {from="", output=""}]
   properties {bUsePreviewValueAsDefault=true, inputName=Index, inputType=FunctionInput_Scalar}
  }
  MaterialExpressionFunctionInput_1 {
   class=MaterialExpressionFunctionInput
   inputs [Preview {from="", output=""}]
   properties {bUsePreviewValueAsDefault=true, inputName=Input_0-3, inputType=FunctionInput_Vector4, sortPriority=1, previewValue=<0 1 2 3>}
  }
  MaterialExpressionFunctionInput_2 {
   class=MaterialExpressionFunctionInput
   inputs [Preview {from="", output=""}]
   properties {bUsePreviewValueAsDefault=true, inputName=Input_4-7, inputType=FunctionInput_Vector4, sortPriority=2, previewValue=<4 5 6 7>}
  }
  MaterialExpressionFunctionInput_3 {
   class=MaterialExpressionFunctionInput
   inputs [Preview {from="", output=""}]
   properties {bUsePreviewValueAsDefault=true, inputName=Input_8-11, inputType=FunctionInput_Vector4, sortPriority=3, previewValue=<8 9 10 11>}
  }
  MaterialExpressionFunctionInput_4 {
   class=MaterialExpressionFunctionInput
   inputs [Preview {from="", output=""}]
   properties {bUsePreviewValueAsDefault=true, inputName=Input_12-15, inputType=FunctionInput_Vector4, sortPriority=4, previewValue=<12 13 14 15>}
  }
  MaterialExpressionFunctionInput_5 {
   class=MaterialExpressionFunctionInput
   inputs [Preview {from=MaterialExpressionStaticBool_0, output=""}]
   properties {bUsePreviewValueAsDefault=true, inputName=Use 12 Inputs, inputType=FunctionInput_StaticBool, sortPriority=10}
  }
  MaterialExpressionFunctionInput_6 {
   class=MaterialExpressionFunctionInput
   inputs [Preview {from=MaterialExpressionStaticBool_0, output=""}]
   properties {bUsePreviewValueAsDefault=true, inputName=Use 16 Inputs, inputType=FunctionInput_StaticBool, sortPriority=11}
  }
  MaterialExpressionFunctionOutput_0 {class=MaterialExpressionFunctionOutput, inputs [None {from=MaterialExpressionSwitch_3, output=""}]}
  MaterialExpressionMaterialFunctionCall_0 {
   class=MaterialExpressionMaterialFunctionCall
   inputs [Float4 {from=MaterialExpressionStaticSwitch_1}]
   properties {materialFunction=&BreakOutFloat4Components}
  }
  MaterialExpressionNamedRerouteDeclaration_0 {
   class=MaterialExpressionNamedRerouteDeclaration
   inputs [None {from=MaterialExpressionFunctionInput_0}]
   properties {name=Index, nodeColor=<rgba 0 0.7891 1 1>}
  }
  MaterialExpressionNamedRerouteUsage_0 {class=MaterialExpressionNamedRerouteUsage}
  MaterialExpressionStaticBool_0 {class=MaterialExpressionStaticBool}
  MaterialExpressionStaticSwitch_0 {
   class=MaterialExpressionStaticSwitch
   inputs [
    True {from=MaterialExpressionSwitch_1, output=""}
    False {from=MaterialExpressionSwitch_0, output=""}
    Value {from=MaterialExpressionFunctionInput_5, output=""}
   ]
   properties {}
  }
  MaterialExpressionStaticSwitch_1 {
   class=MaterialExpressionStaticSwitch
   inputs [
    True {from=MaterialExpressionSwitch_2, output=""}
    False {from=MaterialExpressionStaticSwitch_0, output=""}
    Value {from=MaterialExpressionFunctionInput_6, output=""}
   ]
   properties {}
  }
  MaterialExpressionSwitch_0 {
   class=MaterialExpressionSwitch
   inputs [
    SwitchValue {from=MaterialExpressionFloor_0, output=""}
    Default {from="", output=""}
    None {from=MaterialExpressionFunctionInput_1, output=""}
    None {from=MaterialExpressionFunctionInput_2, output=""}
   ]
   properties {inputs [
     {
      inputName=None
      input {
       expression=&self:MaterialExpressionFunctionInput_1
       inputName=None
       mask=0
       maskA=0
       maskB=0
       maskG=0
       maskR=0
       outputIndex=0
      }
     }
     {
      inputName=None
      input {
       expression=&self:MaterialExpressionFunctionInput_2
       inputName=None
       mask=0
       maskA=0
       maskB=0
       maskG=0
       maskR=0
       outputIndex=0
      }
     }
    ]}
  }
  MaterialExpressionSwitch_1 {
   class=MaterialExpressionSwitch
   inputs [
    SwitchValue {from=MaterialExpressionFloor_0, output=""}
    Default {from="", output=""}
    None {from=MaterialExpressionFunctionInput_1, output=""}
    None {from=MaterialExpressionFunctionInput_2, output=""}
    None {from=MaterialExpressionFunctionInput_3, output=""}
   ]
   properties {inputs [
     {
      inputName=None
      input {
       expression=&self:MaterialExpressionFunctionInput_1
       inputName=None
       mask=0
       maskA=0
       maskB=0
       maskG=0
       maskR=0
       outputIndex=0
      }
     }
     {
      inputName=None
      input {
       expression=&self:MaterialExpressionFunctionInput_2
       inputName=None
       mask=0
       maskA=0
       maskB=0
       maskG=0
       maskR=0
       outputIndex=0
      }
     }
     {
      inputName=None
      input {
       expression=&self:MaterialExpressionFunctionInput_3
       inputName=None
       mask=0
       maskA=0
       maskB=0
       maskG=0
       maskR=0
       outputIndex=0
      }
     }
    ]}
  }
  MaterialExpressionSwitch_2 {
   class=MaterialExpressionSwitch
   inputs [
    SwitchValue {from=MaterialExpressionFloor_0, output=""}
    Default {from="", output=""}
    None {from=MaterialExpressionFunctionInput_1, output=""}
    None {from=MaterialExpressionFunctionInput_2, output=""}
    None {from=MaterialExpressionFunctionInput_3, output=""}
    None {from=MaterialExpressionFunctionInput_4, output=""}
   ]
   properties {inputs [
     {
      inputName=None
      input {
       expression=&self:MaterialExpressionFunctionInput_1
       inputName=None
       mask=0
       maskA=0
       maskB=0
       maskG=0
       maskR=0
       outputIndex=0
      }
     }
     {
      inputName=None
      input {
       expression=&self:MaterialExpressionFunctionInput_2
       inputName=None
       mask=0
       maskA=0
       maskB=0
       maskG=0
       maskR=0
       outputIndex=0
      }
     }
     {
      inputName=None
      input {
       expression=&self:MaterialExpressionFunctionInput_3
       inputName=None
       mask=0
       maskA=0
       maskB=0
       maskG=0
       maskR=0
       outputIndex=0
      }
     }
     {
      inputName=None
      input {
       expression=&self:MaterialExpressionFunctionInput_4
       inputName=None
       mask=0
       maskA=0
       maskB=0
       maskG=0
       maskR=0
       outputIndex=0
      }
     }
    ]}
  }
  MaterialExpressionSwitch_3 {
   class=MaterialExpressionSwitch
   inputs [
    SwitchValue {from=MaterialExpressionFmod_0, output=""}
    Default {from="", output=""}
    None {from=MaterialExpressionMaterialFunctionCall_0, output=R}
    None {from=MaterialExpressionMaterialFunctionCall_0, output=R}
    None {from=MaterialExpressionMaterialFunctionCall_0, output=R}
    None {from=MaterialExpressionMaterialFunctionCall_0, output=R}
   ]
   properties {inputs [
     {
      inputName=None
      input {
       expression=&self:MaterialExpressionMaterialFunctionCall_0
       inputName=None
       mask=0
       maskA=0
       maskB=0
       maskG=0
       maskR=0
       outputIndex=0
      }
     }
     {
      inputName=None
      input {
       expression=&self:MaterialExpressionMaterialFunctionCall_0
       inputName=None
       mask=0
       maskA=0
       maskB=0
       maskG=0
       maskR=0
       outputIndex=1
      }
     }
     {
      inputName=None
      input {
       expression=&self:MaterialExpressionMaterialFunctionCall_0
       inputName=None
       mask=0
       maskA=0
       maskB=0
       maskG=0
       maskR=0
       outputIndex=2
      }
     }
     {
      inputName=None
      input {
       expression=&self:MaterialExpressionMaterialFunctionCall_0
       inputName=None
       mask=0
       maskA=0
       maskB=0
       maskG=0
       maskR=0
       outputIndex=3
      }
     }
    ]}
  }
 }}
