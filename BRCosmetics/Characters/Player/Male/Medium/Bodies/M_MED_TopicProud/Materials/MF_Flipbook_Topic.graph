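# Material function for the TopicProud player body that drives flipbook frames from left and right indices, as either a texture flipbook or a vertex mask.
# Local pre-skinning X separates the sides, with static switches for syncing them and remapping indices through selector functions, optionally idle-only or pose-synced using fractional parts.
# MF_FlipbookTexture_Topic supplies color, alpha and UVs.
# The vertex mask comes from a narrow SphereMask comparing an offset, scaled value to the VertexColor input.
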MaterialFunction /BRCosmetics/Characters/Player/Male/Medium/Bodies/M_MED_TopicProud/Materials/MF_Flipbook_Topic
graph {nodes [
  FunctionInput() {
   bUsePreviewValueAsDefault=true
   description=-------------
   inputName=----------------------------
   inputType=FunctionInput_StaticBool
   sortPriority=30
  }
  FunctionInput() {
   bUsePreviewValueAsDefault=true
   description=-------------
   inputName=-----------------------------
   inputType=FunctionInput_StaticBool
   sortPriority=50
  }
  FunctionInput() {
   bUsePreviewValueAsDefault=true
   description=-------------
   inputName="--------------------------      "
   inputType=FunctionInput_StaticBool
   sortPriority=10
  }
  n1 = NamedRerouteUsage()
  n2 = Add(A=Divide(A=Add(A=@n1)) {constB=20}) {constB=0.05}
  n3 = FunctionInput() {bUsePreviewValueAsDefault=true, inputName=VertexColor, inputType=FunctionInput_Scalar, sortPriority=20}
  FunctionOutput(None=StaticSwitch(True=Constant(), False=SphereMask(A=@n2, B=@n3) {attenuationRadius=0.02}, Value=NamedRerouteUsage())) {outputName=Flipbook Vertex Mask}
  n4 = FunctionInput() {bUsePreviewValueAsDefault=true, inputName=Flipbook Column Count, inputType=FunctionInput_Scalar, sortPriority=42, previewValue=<4 0 0 1>}
  n5 = FunctionInput() {bUsePreviewValueAsDefault=true, inputName=Flipbook Row Count, inputType=FunctionInput_Scalar, sortPriority=43, previewValue=<4 0 0 1>}
  n6 = FunctionInput() {bUsePreviewValueAsDefault=true, inputName=UV Scale, inputType=FunctionInput_Vector2, sortPriority=45, previewValue=<1 1 0 1>}
  n7 = MaterialFunctionCall!(
   Flipbook Index=NamedRerouteUsage()
   Flipbook Texture=FunctionInput(Preview=TextureObject() {texture=&T_BlacksRGB}) {bUsePreviewValueAsDefault=true, inputName=Flipbook Texture, inputType=FunctionInput_Texture2D, sortPriority=41}
   Flipbook Column Count=@n4
   Flipbook Row Count=@n5
   UVs=FunctionInput(Preview=TextureCoordinate()) {bUsePreviewValueAsDefault=true, inputName=UVs, inputType=FunctionInput_Vector2, sortPriority=44}
   UV Offset=FunctionInput() {bUsePreviewValueAsDefault=true, inputName=UV Offset, inputType=FunctionInput_Vector2, sortPriority=46}
   UV Scale=@n6
   MirrorLeftUV=FunctionInput(Preview=StaticBool()) {bUsePreviewValueAsDefault=true, inputName=MirrorLeftUV, inputType=FunctionInput_StaticBool, sortPriority=47}
  ) {materialFunction=&/MF_FlipbookTexture_Topic}
  n8 = Constant()
  n9 = NamedRerouteUsage()
  FunctionOutput(None=StaticSwitch(True=@n7.Color, False=@n8, Value=@n9)) {outputName=FlipbookTexture, sortPriority=1}
  FunctionOutput(None=StaticSwitch(True=@n7.Alpha, False=@n8, Value=@n9)) {outputName=FlipbookTextureAlpha, sortPriority=2}
  FunctionOutput(None=@n7.UV_OUT) {outputName=UV_OUT, sortPriority=10}
  NamedRerouteDeclaration(None=FunctionInput(Preview=StaticBool()) {bUsePreviewValueAsDefault=true, inputName=Use Texture Flipbook, inputType=FunctionInput_StaticBool, sortPriority=40}) {name=IsTextureFlipbook, nodeColor=<rgba 0 0.5781 1 1>}
  NamedRerouteDeclaration(None=FunctionInput(Preview=StaticBool()) {bUsePreviewValueAsDefault=true, inputName=Remap Idle Only, inputType=FunctionInput_StaticBool, sortPriority=61}) {name=RemapIdleOnly, nodeColor=<rgba 1 0.6797 0 1>}
  n10 = FunctionInput() {bUsePreviewValueAsDefault=true, inputName=RemapIndex_12-15, inputType=FunctionInput_Vector4, sortPriority=73, previewValue=<12 13 14 15>}
  NamedRerouteDeclaration(None=@n10) {name=Remap_12-15, nodeColor=<rgba 0.3594 0 1 1>}
  n11 = StaticBool()
  NamedRerouteDeclaration(None=FunctionInput(Preview=@n11) {bUsePreviewValueAsDefault=true, inputName=Remap Use 12 Inputs, inputType=FunctionInput_StaticBool, sortPriority=80}) {name=RemapUse12, nodeColor=<rgba 0.9219 1 0 1>}
  NamedRerouteDeclaration(None=FunctionInput(Preview=@n11) {bUsePreviewValueAsDefault=true, inputName=Remap Use 16 Inputs, inputType=FunctionInput_StaticBool, sortPriority=81}) {name=RemapUse16, nodeColor=<rgba 0.9219 1 0 1>}
  NamedRerouteDeclaration(None=StaticSwitch(True=NamedRerouteUsage(), False=NamedRerouteUsage(), Value=NamedRerouteUsage())) {name=SyncedIndex, nodeColor=<rgba 0.0078 0 1 1>}
  NamedRerouteDeclaration(None=MaterialFunctionCall!(
    Index=NamedRerouteUsage()
    Input_0-3=NamedRerouteUsage()
    Input_4-7=NamedRerouteUsage()
    Input_8-11=NamedRerouteUsage()
    Input_12-15=NamedRerouteUsage()
    Use 12 Inputs=NamedRerouteUsage()
    Use 16 Inputs=NamedRerouteUsage()
   ) {materialFunction=/Game/Global/Materials/Character_Parent_2023/FlipbookFace_Parent_2023/MaterialFunctions/Common/MF_FlipbookIndexSelector}.Result) {name=RemappedSyncedIndex, nodeColor=<rgba 0.3125 1 0 1>}
  n12 = NamedRerouteUsage()
  n13 = NamedRerouteUsage()
  n14 = LinearInterpolate(A=@n12, B=@n13, Alpha=NamedRerouteUsage())
  n15 = NamedRerouteUsage()
  NamedRerouteDeclaration(None=StaticSwitch(True=StaticSwitch(True=@n14, False=@n13, Value=@n15), False=@n12, Value=NamedRerouteUsage())) {name=FinalSyncdIndex, nodeColor=<rgba 0 0.9531 1 1>}
  NamedRerouteDeclaration(None=FunctionInput() {bUsePreviewValueAsDefault=true, inputName=IndexLeft, inputType=FunctionInput_Scalar, sortPriority=1}) {name=IndexLeft, nodeColor=<rgba 0 1 0.25 1>}
  n16 = NamedRerouteDeclaration(None=MaterialFunctionCall(
    Index=NamedRerouteUsage()
    Input_0-3=NamedRerouteUsage()
    Input_4-7=NamedRerouteUsage()
    Input_8-11=NamedRerouteUsage()
    Input_12-15=NamedRerouteUsage()
    Use 12 Inputs=NamedRerouteUsage()
    Use 16 Inputs=NamedRerouteUsage()
   ) {materialFunction=&MF_FlipbookIndexSelector}.Result) {name=RemapIndexRight_1, nodeColor=<rgba 0 1 0.7656 1>}
  n17 = NamedRerouteDeclaration(None=MaterialFunctionCall(
    Index=NamedRerouteUsage()
    Input_0-3=NamedRerouteUsage()
    Input_4-7=NamedRerouteUsage()
    Input_8-11=NamedRerouteUsage()
    Input_12-15=NamedRerouteUsage()
    Use 12 Inputs=NamedRerouteUsage()
    Use 16 Inputs=NamedRerouteUsage()
   ) {materialFunction=&MF_FlipbookIndexSelector}.Result) {name=RemapIndexLeft_1, nodeColor=<rgba 0 1 0.7656 1>}
  n18 = If(A=Frac(None=@n16), B=Frac(None=@n17), "A > B"=@n16, "A == B"=@n16, "A < B"=@n17)
  n19 = LinearInterpolate(A=NamedRerouteUsage(), B=NamedRerouteUsage(), Alpha=NamedRerouteUsage())
  NamedRerouteDeclaration(None=If(A=Frac(None=@n18), "A > B"=@n18, "A == B"=@n19, "A < B"=@n19) {constB=0.49}) {name=PoseSyncdRemappedIndex, nodeColor=<rgba 0 1 0.0859 1>}
  NamedRerouteDeclaration(None=LinearInterpolate(A=NamedRerouteUsage(), B=NamedRerouteUsage(), Alpha=NamedRerouteUsage())) {name=RightLeftIndex, nodeColor=<rgba 1 0 0.2344 1>}
  NamedRerouteDeclaration(None=FunctionInput(Preview=StaticBool()) {bUsePreviewValueAsDefault=true, inputName=UseLeftIndexForSync, inputType=FunctionInput_StaticBool, sortPriority=3}) {name=UseLeftIndexForSync, nodeColor=<rgba 0.875 1 0 1>}
  NamedRerouteDeclaration(None=FunctionInput(Preview=StaticBool()) {bUsePreviewValueAsDefault=true, inputName=LeftRightSync, inputType=FunctionInput_StaticBool, sortPriority=2}) {name=LeftRightSync, nodeColor=<rgba 0.8047 0 1 1>}
  NamedRerouteDeclaration(None=FunctionInput(Preview=StaticBool()) {bUsePreviewValueAsDefault=true, inputName=RemapPoseSync, inputType=FunctionInput_StaticBool, sortPriority=61}) {name=RemapPoseSync, nodeColor=<rgba 1 0.6797 0 1>}
  NamedRerouteDeclaration(None=MaterialFunctionCall(
    Index=NamedRerouteUsage()
    Input_0-3=NamedRerouteUsage()
    Input_4-7=NamedRerouteUsage()
    Input_8-11=NamedRerouteUsage()
    Input_12-15=NamedRerouteUsage()
    Use 12 Inputs=NamedRerouteUsage()
    Use 16 Inputs=NamedRerouteUsage()
   ) {materialFunction=&MF_FlipbookIndexSelector}.Result) {name=RightLeftRemapIndex, nodeColor=<rgba 1 0 0.4922 1>}
  NamedRerouteDeclaration(None=StaticSwitch(True=NamedRerouteUsage(), False=NamedRerouteUsage(), Value=NamedRerouteUsage())) {name=RemappedIndexFinal, nodeColor=<rgba 0 0.8125 1 1>}
  NamedRerouteDeclaration(None=FunctionInput() {bUsePreviewValueAsDefault=true, inputName=IndexRight, inputType=FunctionInput_Scalar}) {name=IndexRight, nodeColor=<rgba 0 1 0.25 1>}
  n20 = LocalPosition() {localOrigin=InstancePreSkinning}
  n21 = ShaderStageSwitch(PixelShader=VertexInterpolator(VS=@n20.XYZ).PS, VertexShader=@n20.XYZ)
  NamedRerouteDeclaration(None=Step(X=MaterialFunctionCall(Float3=@n21) {materialFunction=&BreakOutFloat3Components}.R)) {name=Right0/Left1, nodeColor=<rgba 1 0 0.1172 1>}
  n22 = NamedRerouteUsage()
  n23 = NamedRerouteUsage()
  n24 = NamedRerouteUsage()
  n25 = NamedRerouteDeclaration(None=Floor(None=ScalarParameter() {group=Flipbook 05 Indices, parameterName=flipbook_face_is_idling, sortPriority=100})) {name=IsIdling, nodeColor=<rgba 1 0 0.5156 1>}
  n26 = StaticSwitch(
   True=StaticSwitch(True=LinearInterpolate(A=@n23, B=@n24, Alpha=@n25), False=@n24, Value=NamedRerouteUsage())
   False=NamedRerouteUsage()
   Value=NamedRerouteUsage()
  )
  n27 = NamedRerouteUsage()
  NamedRerouteDeclaration(None=Floor(None=StaticSwitch(True=@n22, False=@n26, Value=@n27))) {name=Index, nodeColor=<rgba 0 1 1 1>}
  NamedRerouteDeclaration(None=FunctionInput(Preview=StaticBool()) {bUsePreviewValueAsDefault=true, inputName=Use Remap, inputType=FunctionInput_StaticBool, sortPriority=60}) {name=UseRemap, nodeColor=<rgba 1 0.6797 0 1>}
  n28 = FunctionInput() {bUsePreviewValueAsDefault=true, inputName=RemapIndex_0-3, inputType=FunctionInput_Vector4, sortPriority=70, previewValue=<0 1 2 3>}
  NamedRerouteDeclaration(None=@n28) {name=Remap_0-3, nodeColor=<rgba 0.3594 0 1 1>}
  n29 = FunctionInput() {bUsePreviewValueAsDefault=true, inputName=RemapIndex_4-7, inputType=FunctionInput_Vector4, sortPriority=71, previewValue=<4 5 6 7>}
  NamedRerouteDeclaration(None=@n29) {name=Remap_4-7, nodeColor=<rgba 0.3594 0 1 1>}
  n30 = FunctionInput() {bUsePreviewValueAsDefault=true, inputName=RemapIndex_8-11, inputType=FunctionInput_Vector4, sortPriority=72, previewValue=<8 9 10 11>}
  NamedRerouteDeclaration(None=@n30) {name=Remap_8-11, nodeColor=<rgba 0.3594 0 1 1>}
 ]}
