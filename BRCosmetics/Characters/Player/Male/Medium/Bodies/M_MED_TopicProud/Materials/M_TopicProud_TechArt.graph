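# Anime-shaded, Cloth-model character material for the TopicProud body, adding animated eyes, tint customization and flipbook facial expressions.
# Shared 2023 anime Pre, Calculations and Post functions form the shading chain, with lit and shaded colors modified between them.
# Eye animation, masked by texture, vertex color or second-UV sign, feeds TintMask-driven tinting, then a face flipbook blended by its alpha and EnableFacialExpression.
# It also writes a parameter-driven Temporal Responsiveness output.
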Material /BRCosmetics/Characters/Player/Male/Medium/Bodies/M_MED_TopicProud/Materials/M_TopicProud_TechArt
graph {
 properties {
  bAutomaticallySetUsageInEditor=false
  bUseMaterialAttributes=true
  bUsedWithClothing=true
  bUsedWithMeshDeformer=true
  bUsedWithMorphTargets=true
  bUsedWithNiagaraMeshParticles=true
  bUsedWithSkeletalMesh=true
  bUsedWithStaticLighting=true
  pixelDepthOffsetMode=PDOM_Legacy
  shadingModel=MSM_Cloth
  usedShadingModels=Cloth
 }
 nodes [
  n1 = MaterialFunctionCall!(CustomUV) {materialFunction=/Game/Global/Materials/Character_Parent_2023/Anime_Parent_2023/MaterialFunctions/MFC_AnimeShader_2023_Pre}
  n2 = MaterialFunctionCall!(CustomLightVector, OverrideNormal_WS, STT=@n1.STT, Normal_Map=@n1.STT, DFL=@n1.STT, FX=@n1.STT) {materialFunction=/Game/Global/Materials/Character_Parent_2023/Anime_Parent_2023/MaterialFunctions/MFC_AnimeShader_2023_Calculations}
  MaterialFunctionCall!(
   Shading_Gradient=@n2.Shading_Gradient
   Threshold_Gradient=@n2.Shading_Gradient
   Specular_Gradient=@n2.Shading_Gradient
   Sheen_Mask=@n2.Shading_Gradient
   Thickness=@n2.Shading_Gradient
   DFL_Mask=@n2.Shading_Gradient
   FX=@n2.Shading_Gradient
   Metallic_Gradient=@n2.Shading_Gradient
   RimLight_Gradient=@n2.Shading_Gradient
   NormalBuffer=@n2.Shading_Gradient
   --------
   CINE_LightSources=@n2.Shading_Gradient
   CINE_LightColor=@n2.Shading_Gradient
   ---------
   Color_Lit=NamedRerouteUsage()
   Color_Shaded=NamedRerouteUsage()
   SRM=@n1.SRM
   Emissive=@n1.SRM
   DFL_Color=@n1.SRM
   Specular_Color=@n1.SRM
   RimLight_Color=@n1.SRM
  ) {materialFunction=/Game/Global/Materials/Character_Parent_2023/Anime_Parent_2023/MaterialFunctions/MFC_AnimeShader_2023_Post}
  n3 = VertexColor()
  n4 = TextureSampleParameter2D() {
   group=01_EyeAnimation_Parameters
   parameterName=EyeMask
   samplerSource=SSM_Wrap_WorldGroupSettings
   samplerType=SAMPLERTYPE_Masks
   sortPriority=-32
   texture=&T_WhiteMask
  }
  n5 = StaticSwitchParameter(True=AppendVector(A=@n3, B=@n3), False=@n4.RGBA) {group=01_EyeAnimation_Parameters, parameterName=UseVertexColorForEyeMask, sortPriority=-33}
  n6 = TextureCoordinate() {coordinateIndex=1}
  n7 = Sign(None=Subtract(A=ComponentMask(None=@n6) {r=true}) {constB=0.0001})
  n8 = MaterialFunctionCall!(
   Color Lit=@n1.Color_Lit
   Color Shaded=@n1.Color_Lit
   Use UV2=StaticBoolParameter() {defaultValue=true, group=01_EyeAnimation_Parameters, parameterName=Use UV2 for Eye Animation, sortPriority=-30}
   Mask=StaticSwitchParameter(
    True=ChannelMaskParameter(None=@n5) {group=01_EyeAnimation_Parameters, parameterName=EyeMaskChannel, sortPriority=-32}.R
    False=Saturate(None=@n7)
   ) {group=01_EyeAnimation_Parameters, parameterName=UseEyeMask, sortPriority=-33}
  ) {materialFunction=/Game/Global/Materials/Character_Parent_2023/Anime_Parent_2023/MaterialFunctions/MF_AnimeEyeAnimation_v2}
  n9 = VectorParameter() {group=Customization, parameterName=LitColor, defaultValue=<rgba 1 0.3617 0 1>}
  n10 = TextureSampleParameter2D() {group=Customization, parameterName=TintMask, samplerType=SAMPLERTYPE_Masks, texture=&T_BlackMask}
  NamedRerouteDeclaration(None=LinearInterpolate(A=@n8.Lit, B=Multiply(A=@n9.RGB, B=@n8.Lit), Alpha=@n10.R)) {name=LitColor, nodeColor=<rgba 1 0.2578 0 1>}
  n11 = NamedRerouteUsage()
  n12 = MaterialFunctionCall!(Material) {materialFunction=&/MF_FlipbookFace_Topic}
  n13 = NamedRerouteDeclaration(None=@n12.FaceAlpha) {name=FlipbookFaceAlpha, nodeColor=<rgba 0.8281 0 1 1>}
  n14 = ScalarParameter() {group=Customization, parameterName=EnableFacialExpression}
  NamedRerouteDeclaration(None=LinearInterpolate(A=@n11, B=LinearInterpolate(A=@n11, B=@n12.FaceColor, Alpha=@n13), Alpha=@n14)) {name=LitOUT, nodeColor=<rgba 0 0.0625 1 1>}
  n15 = Multiply(A=@n9.RGB, B=ScalarParameter() {defaultValue=0.65, group=Customization, parameterName=ShadedMulti}) {constB=0.7}
  NamedRerouteDeclaration(None=LinearInterpolate(A=@n8.Shaded, B=Multiply(A=@n8.Shaded, B=@n15), Alpha=@n10.R)) {name=ShadedColor, nodeColor=<rgba 1 0.6563 0 1>}
  n16 = NamedRerouteUsage()
  n17 = Multiply(A=@n12.FaceColor, B=ScalarParameter() {defaultValue=1, parameterName=FlipbookFaceAnimeShadedValue})
  NamedRerouteDeclaration(None=LinearInterpolate(A=@n16, B=LinearInterpolate(A=@n16, B=@n17, Alpha=@n13), Alpha=@n14)) {name=ShadedOUT, nodeColor=<rgba 0.3359 0 1 1>}
  TemporalResponsivenessOutput!(None=LinearInterpolate(B=ScalarParameter() {parameterName=Temporal Responsiveness}, Alpha=NamedRerouteUsage()) {constA=0.4})
 ]
}
